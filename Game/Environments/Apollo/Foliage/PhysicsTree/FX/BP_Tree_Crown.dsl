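; Blueprint for a physics tree crown effect, driving a Niagara system as the treetop falls and playing a crown-destroy sound.
; On BeginPlay the FallingTreeTop timeline starts, and the sound plays at the AudioLocation component.
; Each timeline update calls SetNSTreeInfo, which writes only FallVector to NS_Tree_Crown, ignoring the computed bend angle, rotation and height inputs.
; UpdateTreeData and the construction script are empty.
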(event EventBeginPlay
  (Utilities|FlowControl|Sequence
    (:then_0
      (bind (_direction _bend_angle _fall_in_z _globalrotate) (Timeline|FallingTreeTop PlayFromStart)
        (:Update
          (CallFunction|SetNSTreeInfo (Variables|Default|GetAxisRotationAngle) (Variables|Default|GetMinHeight) (Variables|Default|GetMaxHeight) (* (* _bend_angle -1.0) 65.0) (Variables|Default|GetFallVector)))))
    (:then_1
      (Audio|PlaySoundatLocation "/Game/Sounds/Artemis/Environment_Actors/PhysicsTrees/PhysicsTree_CrownDestroy_Cue.PhysicsTree_CrownDestroy_Cue" (Transformation|GetWorldLocation (Variables|Default|GetAudioLocation))))))

(event Custom|UpdateTreeData)

(fn SetNSTreeInfo
  (params (AxisRotationAngle :type Float%20%28double-precision%29) (MinHeight :type Float%20%28double-precision%29 :default 1e+03) (MaxHeight :type Float%20%28double-precision%29 :default 4e+03) (BendAngle :type Float%20%28double-precision%29) (FallVector :type Vector))
  (Niagara|SetNiagaraVariable(Vector3) (Variables|Default|GetNS_Tree_Crown) "FallVector" FallVector))

(fn ConstructionScript)
