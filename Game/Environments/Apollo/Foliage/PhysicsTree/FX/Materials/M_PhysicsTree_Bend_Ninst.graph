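# Masked, two-sided foliage material for Niagara mesh particles that bends and dissolves physics tree meshes.
# Dynamic parameters drive a height-masked primary bend, a corrective rotation and a top-pivot lag, with vertex colors masking the lag.
# An age-cubed Z offset, scaled by vertex red, adds to the bend.
# An RGB mask splits opacity between Erode-driven leaf dissolve and particle-alpha noise erosion on the trunk.
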Material /Game/Environments/Apollo/Foliage/PhysicsTree/FX/Materials/M_PhysicsTree_Bend_Ninst
graph {
 properties {bUsedWithNiagaraMeshParticles=true, blendMode=BLEND_Masked, pixelDepthOffsetMode=PDOM_Legacy, twoSided=true, usedShadingModels=Default Lit}
 nodes [
  Constant3Vector()
  Constant3Vector()
  Constant3Vector() {constant=<rgba 0.05 0.25 0.05 0>}
  Constant() {r=1}
  n1 = VertexColor() {desc=SH baked lighting}
  n2 = TextureCoordinate()
  n3 = TextureSampleParameter2D(UVs=@n2) {parameterName=Diffuse, texture=/Game/Environments/Apollo/Foliage/PhysicsTree/Textures/T_PhysicsTree_Branch_01_D}
  Multiply(
   A=Multiply(A=Multiply(A=@n1, B=@n3.RGB), B=VectorParameter() {parameterName=ColorMult, defaultValue=<rgba 1 1 1 0>}.RGB)
   B=ScalarParameter() {defaultValue=1, parameterName=SSS_Scale, sliderMax=1}
  )
  n4 = Constant3Vector() {constant=<rgba 1 0 0 0>}
  n5 = ScalarParameter() {defaultValue=90, parameterName=AxisRotationAngle, sliderMax=360}
  n6 = RotateAboutAxis(
   NormalizedRotationAxis=Constant3Vector() {constant=<rgba 0 0 1 0>}
   RotationAngle=NamedRerouteDeclaration(None=Divide(A=@n5) {constB=360}) {name=AxisRotationAngle, nodeColor=<rgba 1 0.3281 0 1>}
   PivotPoint=Constant3Vector()
   Position=@n4
  )
  NamedRerouteDeclaration(None=Normalize(VectorInput=Add(A=@n6, B=@n4))) {name=BendAxis, nodeColor=<rgba 0 0.3906 1 1>}
  ParticleColor()
  n7 = NamedRerouteUsage()
  n8 = DynamicParameter() {parameterIndex=1, defaultValue=<rgba 0 1.4 0 1>, paramNames [LagAngle, LocalFix, NormalizedAge, Erode]}
  n9 = MaterialFunctionCall() {materialFunction=&LocalPosition}
  n10 = MaterialFunctionCall(Float3=@n9."Instance Local Position (Excluding Offsets)") {materialFunction=&BreakOutFloat3Components}
  n11 = DynamicParameter() {defaultValue=<rgba 0 1000 4000 0>, paramNames [AxisRotationAngle, MinHeight, MaxHeight, BendAngle]}
  n12 = Clamp(None=MaterialFunctionCall(Input=@n10.B, Input Low=@n11.MinHeight, Input High=@n11.MaxHeight, Target Low=Constant(), Target High=Constant() {r=1}) {materialFunction=&RemapValueRange}.Result)
  n13 = Multiply(
   A=Power(Base=@n12, Exp=ScalarParameter() {defaultValue=1, parameterName=Exponent, sliderMax=5})
   B=Divide(A=Multiply(A=@n11.BendAngle) {constB=-90}) {constB=360}
  )
  n14 = Multiply(A=@n13, B=@n8.LocalFix) {constB=-0.1}
  n15 = Constant()
  n16 = MaterialFunctionCall(X=@n15, Y=@n15, Z=@n10.B) {materialFunction=&MakeFloat3}
  n17 = NamedRerouteUsage()
  n18 = MaterialFunctionCall(Z=@n11.MinHeight) {materialFunction=&MakeFloat3}
  n19 = OneMinus(None=@n12)
  n20 = Constant()
  n21 = RotateAboutAxis(
   NormalizedRotationAxis=NamedRerouteUsage()
   RotationAngle=Multiply(A=Multiply(A=@n19) {constB=0.01}, B=Divide(A=@n8.LagAngle) {constB=1}) {constB=-0.1}
   PivotPoint=Constant3Vector()
   Position=Subtract(
    A=@n9."Instance Local Position (Excluding Offsets)"
    B=MaterialFunctionCall(X=@n20, Y=@n20, Z=@n11.MaxHeight) {desc=Top Pivot, materialFunction=&MakeFloat3}.Result
   )
  ) {desc=rotate by axis}
  n22 = VertexColor() {desc="mask the trunk with green\r\nred = random\r\nblue = trunk mask\r\nblue = radial erode gradient"}
  n23 = OneMinus(None=@n22.G)
  n24 = Add(
   A=Add(
    A=RotateAboutAxis(NormalizedRotationAxis=@n17, RotationAngle=@n13, PivotPoint=@n18.Result, Position=@n9."Instance Local Position (Excluding Offsets)") {desc=Primary Bend}
    B=RotateAboutAxis(NormalizedRotationAxis=@n7, RotationAngle=@n14, PivotPoint=@n16.Result, Position=@n9."Instance Local Position (Excluding Offsets)") {desc=rotate by axis}
   )
   B=Multiply(A=Multiply(A=@n21, B=@n23), B=@n22.R)
  )
  n25 = Multiply(A=@n22.R, B=Multiply(A=Power(Base=@n8.NormalizedAge) {constExponent=3}) {constB=100})
  Reroute(None=Add(
    A=Transform(None=@n24) {transformSourceType=TRANSFORMSOURCE_Local}
    B=MaterialFunctionCall(Z=@n25) {materialFunction=&MakeFloat3}.Result
   ))
  n26 = TextureSampleParameter2D(UVs=@n2) {
   parameterName=Normals
   samplerSource=SSM_Wrap_WorldGroupSettings
   samplerType=SAMPLERTYPE_Normal
   texture=/Game/Environments/Apollo/Foliage/Textures_Materials/T_Oak_Leaves_N
  }
  Reroute(None=@n26.RGB)
  n27 = TextureSampleParameter2D(UVs=@n2) {
   parameterName=Mask
   samplerType=SAMPLERTYPE_Masks
   texture=/Game/Environments/Apollo/Foliage/PhysicsTree/Textures/T_PhysicsTree_Branch_01_Mask
  }
  Reroute(None=@n27.B)
  Reroute(None=@n27.G)
  n28 = TextureSampleParameter2D(UVs=TextureCoordinate()) {
   desc="mask the trunk with green\r\nred = random\r\ngrn = trunk mask\r\nblue = radial erode gradient"
   parameterName=RGBMask
   texture=&/T_TreeCrown_Leaves_RGB
  }
  n29 = Multiply(A=@n28.R) {constB=-0.3}
  n30 = LinearInterpolate(A=Multiply(A=Power(Base=@n28.R) {constExponent=1.3}), B=Multiply(A=Power(Base=@n28.B) {constExponent=1}))
  n31 = DynamicParameter() {parameterIndex=1, defaultValue=<rgba 0 1.4 0 1>, paramNames [LagAngle, LocalFix, NormalizedAge, Erode]}
  n32 = Subtract(
   A=Add(A=Add(A=@n29, B=@n30))
   B=LinearInterpolate(A=ScalarParameter() {defaultValue=1.2, parameterName=DissolveRate, sliderMax=2}, Alpha=@n31.Erode) {constA=1.2, constB=0}
  )
  n33 = LinearInterpolate(A=@n27.R, B=Constant3Vector() {constant=<rgba 1 0 0 0>}, Alpha=ScalarParameter() {parameterName=fullOpaque, sliderMax=1})
  n34 = Multiply(A=Clamp(None=Floor(None=@n32)), B=@n33)
  n35 = TextureSampleParameter2D(UVs=TextureCoordinate()) {
   desc="mask the trunk with green\r\nred = random\r\ngrn = trunk mask\r\nblue = radial erode gradient"
   parameterName=RGBMask
   texture=&T_TreeCrown_Leaves_RGB
  }
  n36 = ScalarParameter() {defaultValue=1, group=Erosion, parameterName=Erosion UV Scale U}
  n37 = ScalarParameter() {defaultValue=1, group=Erosion, parameterName=Erosion UV Scale V}
  n38 = TextureSampleParameter2D(UVs=Multiply(A=TextureCoordinate(), B=AppendVector(A=@n36, B=@n37))) {group=Erosion, parameterName=ErosionTextureMask, samplerType=SAMPLERTYPE_LinearColor, texture=/Game/Global/Textures/Default/Noise/Noise32x32}
  n39 = Add(A=LinearInterpolate(A=@n38.R, B=OneMinus(None=@n38.R), Alpha=ScalarParameter() {group=Erosion, parameterName=InvertErode, sliderMin=1}))
  n40 = LinearInterpolate(A=ScalarParameter() {defaultValue=1.2, group=Erosion, parameterName=ErodeStart, sliderMax=2}, Alpha=ParticleColor().A) {constA=1.2, constB=0}
  n41 = Clamp(None=Floor(None=Subtract(A=@n39, B=@n40)))
  Reroute(None=Clamp(None=LinearInterpolate(A=@n34, B=@n41, Alpha=@n35.G)))
  ScalarParameter() {parameterName=BendAngle, sliderMax=180}
  ScalarParameter() {defaultValue=1500, parameterName=MinHeight}
  ScalarParameter() {defaultValue=1.45, parameterName=LocalFix, sliderMax=3, sliderMin=-3}
  ScalarParameter() {parameterName=LagAngle, sliderMax=360, sliderMin=-360}
  ScalarParameter() {defaultValue=4000, parameterName=MaxHeight}
  ScalarParameter() {parameterName=UseAlphaErode, sliderMax=1}
  ScalarParameter() {parameterName=DebugErode, sliderMax=1}
  n42 = Multiply(A=@n35.R)
  Subtract(A=@n33, B=Subtract(A=Multiply(A=@n31.Erode) {constB=0.75}, B=Add(A=@n42, B=@n35.B)))
 ]
}
