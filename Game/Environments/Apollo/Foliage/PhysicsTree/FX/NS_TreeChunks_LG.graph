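# Niagara effect for physics trees that throws a burst of short-lived mesh chunks, presumably large tree debris.
# Its one emitter spawns chunks in a cylinder with random non-uniform scale and spin, launched away from a point below the origin.
# Drag and strong gravity pull them down while a curve shrinks them late in life.
# Impact-driven velocity modules are disabled, and no shown module reads the fall, bend or height user parameters.
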NiagaraSystem /Game/Environments/Apollo/Foliage/PhysicsTree/FX/NS_TreeChunks_LG
dependencies {
 usedDataInterfaces [&NiagaraDataInterfaceMeshRendererInfo, &NiagaraDataInterfaceCurve]
 usedDynamicInputs []
 usedModules [&SystemState]
}
emitters {MeshRotationForce001 {
  emitterUpdateScript [
   &EmitterState(Life Cycle Mode=::NewEnumerator1, Loop Behavior=::NewEnumerator1, Loop Duration=3, UseLoopDelay=false)
   &SpawnBurst_Instantaneous(Spawn Count=12, Use Spawn Probability=false, UseLoopCountLimit=false) {module=SpawnBurst_Instantaneous001}
  ]
  particleSpawnScript [
   &InitializeParticle(
    Color Mode=::NewEnumerator0
    Lifetime=&RandomRangeFloat(Minimum=0.2, Maximum=1.2)
    Position Mode=::NewEnumerator0
    Write Material Random=true
   )
   &CylinderLocation(
    Cylinder Height=200
    Cylinder Radius=75
    MaskSpawnGroup=false
    Offset Mode:ENiagara_OffsetMode=::NewEnumerator0
    Override Seed=true
    Random Seed=125
    Randomness Mode:ENiagaraRandomnessMode=::NewEnumerator1
    UseOwnerScale=true
   )
   &self:Random_Scale_XYZ(MaxX=0.2, MaxY=0.6, MaxZ=0.2, MinX=0.7, MinY=1, MinZ=1.5) {module=Random_Scale_XYZ}
   &InitialMeshOrientation(Use Orientation Vector=true, Use Rotation Vector=false)
   &MeshRotationForce(Rotation=<0 25 0>) {enabled=false}
   &AddRotationalVelocity(Rotation Rate=&RandomVector(Vector Scale=&RandomRangeFloat(Minimum=2, Maximum=5)))
   /SetVariables(Particles.ImpactN:Vector3f=<1 1 0>, Particles.ImpactP:Vector3f=<5 0 0>) {enabled=false, isSetParameters=true}
   &self:set_v_from_ImpactN(Falloff Exponent=1, ImpactN:Vector3f=<1 0 0>, ImpactP:Vector3f=<-5 0 0>, Linear Vel Scale=1255, Torque=555) {enabled=false, module=set_v_from_ImpactN}
   &AddVelocityFromPoint(
    Module.Use Falloff Distance=false
    Origin Offset=<0 0 -100>
    Velocity Strength=&RandomRangeFloat(Minimum=555, Maximum=4444)
   )
   &ApplyInitialForces()
  ]
  particleUpdateScript [
   &ParticleState()
   /SetVariables(Transient.PhysicsRotationalDrag:NiagaraFloat=&RandomRangeFloat(Minimum=1, Maximum=2)) {isSetParameters=true}
   &Drag(Drag=&RandomRangeFloat(Minimum=1, Maximum=3), Use Linear Drag=true, Use Rotational Drag=false)
   &ScaleMeshSize(Scale Factor=&VectorFromFloat(Value=&FloatFromCurve(
      FloatCurve=DataInterface() {propertyValues {Curve {keys [<at 0.7 1>,<at 1 0>]}}}
      Scale Curve=&RandomRangeFloat(Minimum=0.5, Maximum=2)
     )))
   &GravityForce(Gravity=<0 0 -3000>)
   &SolveForcesAndVelocity(Clamp Velocity=false, Limit Acceleration=false)
   &SolveRotationalForcesAndVelocity()
  ]
  rendererClasses [&NiagaraMeshRendererProperties]
  renderers [{rendererClass=&NiagaraMeshRendererProperties, rendererIndex=0}]
 }}
summary {
 systemName=NS_TreeChunks_LG
 emitters [{emitterName=MeshRotationForce001, rendererClasses [&NiagaraMeshRendererProperties]}]
 userVariables [
  User.AxisRotationAngle {type=NiagaraFloat, defaultValue=0, description=""}
  User.BendAngle {type=NiagaraFloat, defaultValue=0, description=""}
  User.FallVector {type=Vector3f, description="", defaultValue {struct=Vector3f, value=<1 0 0>}}
  User.MaxHeight {type=NiagaraFloat, defaultValue=1000, description=""}
  User.MinHeight {type=NiagaraFloat, defaultValue=4000, description=""}
 ]
}
