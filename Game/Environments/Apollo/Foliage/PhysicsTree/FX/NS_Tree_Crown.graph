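# Niagara mesh effect for a falling PhysicsTree crown that breaks into branches, leaves and splinters.
# Tree_Instance001 hops, falls and rotates about an axis derived from User.FallVector, driving material parameters with angle, height bounds and age while fading out.
# One-shot bursts place branch chunks on a tapering spiral, leaves in a high sphere and splinters in a trunk cylinder.
# Debris flies from offset points under gravity and drag; leaves also follow FallVector.
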NiagaraSystem /Game/Environments/Apollo/Foliage/PhysicsTree/FX/NS_Tree_Crown
dependencies {
 usedDataInterfaces [
  &NiagaraDataInterfaceCurve
  &NiagaraDataInterfaceColorCurve
  &NiagaraDataInterfaceMeshRendererInfo
  &NiagaraDataInterfaceCurlNoise
 ]
 usedDynamicInputs []
 usedModules [&SystemState]
}
emitters {
 Branchy_Chunks {
  emitterUpdateScript [
   &EmitterState()
   &SpawnBurst_Instantaneous(Spawn Count=12, Use Spawn Probability=false, UseLoopCountLimit=false)
  ]
  particleSpawnScript [
   &InitializeParticle(
    Color Mode=::NewEnumerator0
    Lifetime Max=3.75
    Lifetime Min=2
    Lifetime Mode=::NewEnumerator1
    Mass Max=3
    Mass Mode=::NewEnumerator3
    Mesh Scale Mode=::NewEnumerator3
    Mesh Uniform Scale=1.25
    UsePositionOffset=false
    Write Material Random=true
   )
   &InitialMeshOrientation(
    Mesh Orientation Mode=::NewEnumerator6
    Orientation Quaternion:Quat4f=&self:ScratchDynamicInput_01(
     A:NiagaraFloat=&FloatFromCurve(
      FloatCurve=DataInterface() {propertyValues {Curve {keys [<at 0 1>,<at 1 50>]}}}
      CurveIndex=&ReturnNormalizedExecIndex()
     )
     W:NiagaraFloat=&ReturnNormalizedExecIndex(Normalized Index Scale=3240, Normalized Index Includes One=false)
    )
    Use Rotation Vector=false
   )
   /SetVariables(
    Particles.MeshIndex=&RandomRangeInt(Minimum=0, Maximum=1)
    Particles.Position=&Add_Vector(
     A=/Niagara/DynamicInputs/Helpers/SimulationPosition()
     B=&MakeVector(
      X=&Sine(
       Normalized Angle=&ReturnNormalizedExecIndex()
       Period=1
       Scale=&FloatFromCurve(
        FloatCurve=DataInterface() {propertyValues {Curve {keys [{tangentMode=RCTM_User, time=0, value=150}, {tangentMode=RCTM_User, time=1, value=0}]}}}
        CurveIndex=&ReturnNormalizedExecIndex()
       )
       Bias=0
      )
      Y=&Cosine(
       Normalized Angle=&ReturnNormalizedExecIndex()
       Period=1
       Scale=&FloatFromCurve(
        FloatCurve=DataInterface() {propertyValues {Curve {keys [{tangentMode=RCTM_User, time=0, value=150}, {tangentMode=RCTM_User, time=1, value=0}]}}}
        CurveIndex=&ReturnNormalizedExecIndex()
       )
       Bias=0
      )
      Z=&ReturnNormalizedExecIndex(Normalized Index Scale=1500)
     )
    )
    Particles.Scale=&VectorFromFloat(Value=&FloatFromCurve(
      FloatCurve=DataInterface() {propertyValues {Curve {keys [{tangentMode=RCTM_User, time=0, value=1.4}, {tangentMode=RCTM_User, time=1, value=0.8}]}}}
      CurveIndex=&ReturnNormalizedExecIndex()
      Scale Curve=0.85
     ))
   ) {isSetParameters=true}
   /SetVariables(Particles.Position=&Add_Vector(A=Linked() {linkedVariable {name=Particles.Position, type=NiagaraPosition}}, B=<0 0 1500>)) {isSetParameters=true}
   &AddRotationalVelocity(Rotation Rate=&RandomVector(Vector Scale=&RandomRangeFloat(Minimum=0.1, Maximum=0.45)))
   &AddVelocityFromPoint(
    Module.Use Falloff Distance=false
    Origin Offset=<0 0 1100>
    Velocity Strength=&RandomRangeFloat(Minimum=125, Maximum=2222)
   )
   &CurlNoiseForce(Noise Frequency=0.7, Noise Strength=2000, Pan Noise=false) {module=CurlNoiseForce001}
   &AddVelocity(Velocity=&Multiply_VectorByFloat(Vector=<0 0 1>, Float=&RandomRangeFloat(Minimum=0, Maximum=200)))
   &ApplyInitialForces()
  ]
  particleUpdateScript [
   &ParticleState()
   /SetVariables(Transient.PhysicsRotationalDrag:NiagaraFloat=&RandomRangeFloat(Minimum=0.2, Maximum=2)) {isSetParameters=true}
   &Drag(Drag=&RandomRangeFloat(Minimum=0.5, Maximum=1), Use Linear Drag=true, Use Rotational Drag=false)
   &ScaleMeshSize(Scale Factor=&VectorFromFloat(Value=&FloatFromCurve(
      FloatCurve=DataInterface() {propertyValues {Curve {keys [
          {interpMode=RCIM_Cubic, leaveTangentWeight=0.3655, tangentMode=RCTM_User, time=0.15, value=1}
          {arriveTangent=-8.0882, arriveTangentWeight=0.3655, interpMode=RCIM_Cubic, leaveTangent=-8.0882, tangentMode=RCTM_User, time=0.3, value=0}
         ]}}}
      Scale Curve=0.86
     ))) {enabled=false}
   &GravityForce(Gravity=<0 0 -2500>)
   &CurlNoiseForce(Noise Frequency=0.5, Noise Strength=1555, Pan Noise=false)
   &SolveForcesAndVelocity(Clamp Velocity=false, Limit Acceleration=false)
   &SolveRotationalForcesAndVelocity()
   &ScaleColor(
    Scale Alpha=&FloatFromCurve(FloatCurve=DataInterface() {propertyValues {Curve {keys [<at 0.15 1>,<at 0.35 0>]}}})
    ScaleA=true
    ScaleRGB=false
   )
  ]
  rendererClasses [&NiagaraMeshRendererProperties]
  renderers [{rendererClass=&NiagaraMeshRendererProperties, rendererIndex=0}]
 }
 SmallerLeaves {
  emitterUpdateScript [
   &EmitterState()
   &SpawnBurst_Instantaneous(Spawn Count=400, Use Spawn Probability=false, UseLoopCountLimit=false) {module=SpawnBurst_Instantaneous001}
  ]
  particleSpawnScript [
   &InitializeParticle(
    Color Mode=::NewEnumerator0
    Lifetime=&RandomRangeFloat(Minimum=1.25, Maximum=2.75)
    Mass Max=2
    Mass Mode=::NewEnumerator3
    Position Mode=::NewEnumerator0
    Write Material Random=true
   )
   &SphereLocation(
    MaskSpawnGroup=false
    Offset=<0 0 2000>
    Override Seed=false
    Sphere Radius=555
    Surface Only=false
    UseNonUniformScale=true
    UseOffset=true
    UseSphereOrientation=true
   )
   &CylinderLocation(
    Cylinder Height=200
    Cylinder Radius=75
    MaskSpawnGroup=false
    Offset=<0 0 1200>
    Offset Mode:ENiagara_OffsetMode=::NewEnumerator0
    Override Seed=true
    Random Seed=125
    Randomness Mode:ENiagaraRandomnessMode=::NewEnumerator1
    UseOwnerScale=true
   ) {enabled=false}
   &self:Random_Scale_XYZ(MaxX=0.2, MaxY=0.6, MaxZ=0.2, MinX=0.7, MinY=1, MinZ=1.5) {enabled=false, module=Random_Scale_XYZ}
   &self:Random_Scale_XYZ(MaxX=0.2, MaxY=0.6, MaxZ=0.2, MinX=1, MinY=1.3, MinZ=2) {module=Random_Scale_XYZ001}
   &InitialMeshOrientation(Use Orientation Vector=true, Use Rotation Vector=false)
   &MeshRotationForce(Rotation=<0 25 0>) {enabled=false}
   &AddRotationalVelocity(Rotation Rate=&RandomVector(Vector Scale=&RandomRangeFloat(Minimum=2, Maximum=5)))
   /SetVariables(Particles.ImpactN:Vector3f=<1 1 0>, Particles.ImpactP:Vector3f=<5 0 0>) {enabled=false, isSetParameters=true, module=SetVariables_#1}
   &self:set_v_from_ImpactN(Falloff Exponent=1, ImpactN:Vector3f=<1 0 0>, ImpactP:Vector3f=<-5 0 0>, Linear Vel Scale=1255, Torque=555) {enabled=false, module=set_v_from_ImpactN}
   &AddVelocityFromPoint(
    Module.Use Falloff Distance=false
    Origin Offset=&MakeVector(X=0, Y=0, Z=&RandomRangeFloat(Minimum=2400, Maximum=1600))
    Velocity Strength=&RandomRangeFloat(Minimum=222, Maximum=2222)
   )
   &AddVelocity(
    Coordinate Space=::World
    Velocity=&Multiply_VectorByFloat(Vector=$User.FallVector, Float=&RandomRangeFloat(Minimum=2222, Maximum=1111))
   )
   &ApplyInitialForces()
  ]
  particleUpdateScript [
   &ParticleState()
   /SetVariables(Transient.PhysicsRotationalDrag:NiagaraFloat=&RandomRangeFloat(Minimum=1, Maximum=2)) {isSetParameters=true, module=SetVariables_#2}
   &Drag(Drag=&RandomRangeFloat(Minimum=1, Maximum=3), Use Linear Drag=true, Use Rotational Drag=false)
   &ScaleMeshSize(Scale Factor=&VectorFromFloat(Value=&FloatFromCurve(FloatCurve=DataInterface() {propertyValues {Curve {keys [<at 0.7 1>,<at 1 0>]}}}, Scale Curve=5)))
   &GravityForce(Gravity=<0 0 -2000>)
   &CurlNoiseForce(Noise Frequency=1, Noise Strength=555, Pan Noise=false)
   &SolveForcesAndVelocity(Clamp Velocity=false, Limit Acceleration=false)
   &SolveRotationalForcesAndVelocity()
  ]
  rendererClasses [&NiagaraMeshRendererProperties]
  renderers [{rendererClass=&NiagaraMeshRendererProperties, rendererIndex=0}]
 }
 SplinteryBits {
  emitterUpdateScript [
   &EmitterState()
   &SpawnBurst_Instantaneous(Spawn Count=59, Use Spawn Probability=false, UseLoopCountLimit=false) {module=SpawnBurst_Instantaneous001}
  ]
  particleSpawnScript [
   &InitializeParticle(
    Color Mode=::NewEnumerator0
    Lifetime=&RandomRangeFloat(Minimum=0.2, Maximum=2)
    Position Mode=::NewEnumerator0
    Write Material Random=true
   )
   &CylinderLocation(
    Cylinder Height=200
    Cylinder Radius=75
    MaskSpawnGroup=false
    Offset=<0 0 1200>
    Offset Mode:ENiagara_OffsetMode=::NewEnumerator0
    Override Seed=true
    Random Seed=125
    Randomness Mode:ENiagaraRandomnessMode=::NewEnumerator1
    UseOwnerScale=true
   )
   &self:Random_Scale_XYZ(MaxX=0.2, MaxY=0.6, MaxZ=0.2, MinX=0.7, MinY=1, MinZ=1.5) {module=Random_Scale_XYZ}
   &InitialMeshOrientation(Use Orientation Vector=true, Use Rotation Vector=false)
   &MeshRotationForce(Rotation=<0 25 0>) {enabled=false}
   &AddRotationalVelocity(Rotation Rate=&RandomVector(Vector Scale=&RandomRangeFloat(Minimum=2, Maximum=5)))
   /SetVariables(Particles.ImpactN:Vector3f=<1 1 0>, Particles.ImpactP:Vector3f=<5 0 0>) {enabled=false, isSetParameters=true, module=SetVariables_#1}
   &self:set_v_from_ImpactN(Falloff Exponent=1, ImpactN:Vector3f=<1 0 0>, ImpactP:Vector3f=<-5 0 0>, Linear Vel Scale=1255, Torque=555) {enabled=false, module=set_v_from_ImpactN}
   &AddVelocityFromPoint(
    Module.Use Falloff Distance=false
    Origin Offset=<0 0 1100>
    Velocity Strength=&RandomRangeFloat(Minimum=125, Maximum=4444)
   )
   &ApplyInitialForces()
  ]
  particleUpdateScript [
   &ParticleState()
   /SetVariables(Transient.PhysicsRotationalDrag:NiagaraFloat=&RandomRangeFloat(Minimum=1, Maximum=2)) {isSetParameters=true, module=SetVariables_#2}
   &Drag(Drag=&RandomRangeFloat(Minimum=1, Maximum=3), Use Linear Drag=true, Use Rotational Drag=false)
   &ScaleMeshSize(Scale Factor=&VectorFromFloat(Value=&FloatFromCurve(FloatCurve=DataInterface() {propertyValues {Curve {keys [<at 0.7 1>,<at 1 0>]}}}, Scale Curve=1.75)))
   &GravityForce(Gravity=<0 0 -3000>)
   &SolveForcesAndVelocity(Clamp Velocity=false, Limit Acceleration=false)
   &SolveRotationalForcesAndVelocity()
  ]
  rendererClasses [&NiagaraMeshRendererProperties]
  renderers [{rendererClass=&NiagaraMeshRendererProperties, rendererIndex=0}]
 }
 Tree_Instance001 {
  emitterUpdateScript [
   &EmitterState()
   &SpawnBurst_Instantaneous(Use Spawn Probability=false, UseLoopCountLimit=false)
  ]
  particleSpawnScript [
   &InitializeParticle(
    Color Mode=::NewEnumerator0
    Lifetime=4
    Mesh Scale Mode=::NewEnumerator3
    Position=<0 0 3000>
    Position Mode=::NewEnumerator1
    Position Offset Coordinate Space=::Local
    UsePositionOffset=true
    Write Material Random=true
   )
   &AddVelocity(Coordinate Space=::World, Velocity=<0 0 550>)
   &InitialMeshOrientation(Mesh Orientation Mode=::NewEnumerator2, Orientation Coordinate Space=::Local, Use Rotation Vector=false) {enabled=false}
  ]
  particleUpdateScript [
   &GravityForce(Gravity=<0 0 -1700>)
   &Drag(Use Linear Drag=true, Use Rotational Drag=false)
   &ParticleState()
   &UpdateMeshOrientation(
    Rotation Coordinate Space=::NewEnumerator2
    Rotation Rate=&FloatFromCurve(
     FloatCurve=DataInterface() {propertyValues {Curve {keys [{interpMode=RCIM_Cubic, leaveTangent=1.8598, tangentMode=RCTM_Break, time=0, value=0}, <at 1 1>]}}}
     Scale Curve=0.35
    )
    Rotation Vector=&NormalizeVector(
     Vector To Normalize=/Niagara/DynamicInputs/Vectors/CrossProduct(
      Vector A:Vector3f=&TransformVector(
       Vector=&NormalizeVector(Vector To Normalize=$User.FallVector, Safe Normalize=true, Safe Fallback Threshold=0, Safe Fallback Vector=<0 0 0.001>)
       Source Space=::World
       Destination Space=::Local
      )
      Vector B:Vector3f=<0 0 1>
      Normalize Incoming Vectors:NiagaraBool=true
     )
     Safe Normalize=true
     Safe Fallback Threshold=0
     Safe Fallback Vector=<0 0 0.001>
    )
   ) {module=UpdateMeshOrientation002}
   &self:ScratchModule(
    EngineAxis:Vector3f=&TransformVector(Vector=$User.FallVector, Source Space=::World, Destination Space=::Local)
    Orient:Quat4f=$Particles.MeshOrientation
   ) {module=ScratchModule}
   &DynamicMaterialParameters(
    Index 0 Param 1=$Particles.AxisRotAngle
    Index 0 Param 2=$User.MinHeight
    Index 0 Param 3=$User.MaxHeight
    Index 0 Param 4=&Multiply_Float(A=$User.BendAngle, B=-2)
    Index 1 Param 1=&Multiply_Float(A=$User.BendAngle, B=0.1)
    Index 1 Param 2=2
    Index 1 Param 3=$Particles.NormalizedAge
    Index 1 Param 4=&FloatFromCurve(FloatCurve=DataInterface() {propertyValues {Curve {keys [
         {arriveTangent=-4.5528, interpMode=RCIM_Cubic, leaveTangent=-4.5528, tangentMode=RCTM_User, time=0.05, value=1}
         {arriveTangent=-0.4554, interpMode=RCIM_Cubic, leaveTangent=-0.4554, tangentMode=RCTM_User, time=0.5, value=0.1}
         {arriveTangent=-0.2, interpMode=RCIM_Cubic, leaveTangent=-0.2, time=1, value=0}
        ]}}})
    Module.Param0WriteEnabled=true
    Module.Param0WriteEnabled1=true
    Module.Param1WriteEnabled=true
    Module.Param1WriteEnabled1=true
    Module.Param2WriteEnabled=true
    Module.Param2WriteEnabled1=true
    Module.Param3WriteEnabled=true
    Module.Param3WriteEnabled1=true
    Write Parameter Index 1=true
   ) {enabled=false}
   &DynamicMaterialParameters(
    Index 0 Param 1=$Particles.AxisRotAngle
    Index 0 Param 2=$User.MinHeight
    Index 0 Param 3=$User.MaxHeight
    Index 0 Param 4=&Multiply_Float(A=$Emitter.Age, B=-0.1)
    Index 1 Param 1=&Multiply_Float(A=1, B=0.3)
    Index 1 Param 2=2
    Index 1 Param 3=$Particles.NormalizedAge
    Index 1 Param 4=&FloatFromCurve(FloatCurve=DataInterface() {propertyValues {Curve {keys [
         {arriveTangent=-4.5528, interpMode=RCIM_Cubic, leaveTangent=-4.5528, tangentMode=RCTM_User, time=0.05, value=1}
         {arriveTangent=-0.4554, interpMode=RCIM_Cubic, leaveTangent=-0.4554, tangentMode=RCTM_User, time=0.5, value=0.1}
         {arriveTangent=-0.2, interpMode=RCIM_Cubic, leaveTangent=-0.2, time=1, value=0}
        ]}}})
    Module.Param0WriteEnabled=true
    Module.Param0WriteEnabled1=true
    Module.Param1WriteEnabled=true
    Module.Param1WriteEnabled1=true
    Module.Param2WriteEnabled=true
    Module.Param2WriteEnabled1=true
    Module.Param3WriteEnabled=true
    Module.Param3WriteEnabled1=true
    Write Parameter Index 1=true
   ) {module=DynamicMaterialParameters001}
   &SolveForcesAndVelocity(Clamp Velocity=false, Limit Acceleration=false)
   &Color(Color=&ColorFromCurve(ColorCurve=DataInterface() {propertyValues {
       AlphaCurve {keys [
         {arriveTangent=-0.4273, interpMode=RCIM_Cubic, leaveTangent=-0.4273, tangentMode=RCTM_User, time=0.0384, value=1}
         {arriveTangent=-6.051, interpMode=RCIM_Cubic, leaveTangent=-6.051, tangentMode=RCTM_User, time=0.2928, value=0}
        ]}
       BlueCurve {keys [<at 0 1 RCIM_Cubic>,<at 1 1 RCIM_Cubic>]}
       GreenCurve {keys [<at 0 1 RCIM_Cubic>,<at 1 1 RCIM_Cubic>]}
       RedCurve {keys [<at 0 1 RCIM_Cubic>,<at 1 1 RCIM_Cubic>]}
      }}))
  ]
  rendererClasses [&NiagaraMeshRendererProperties]
  renderers [{rendererClass=&NiagaraMeshRendererProperties, rendererIndex=0}]
 }
}
summary {
 systemName=NS_Tree_Crown
 emitters [
  {emitterName=Tree_Instance001, rendererClasses [&NiagaraMeshRendererProperties]}
  {emitterName=SplinteryBits, rendererClasses [&NiagaraMeshRendererProperties]}
  {emitterName=Branchy_Chunks, rendererClasses [&NiagaraMeshRendererProperties]}
  {emitterName=SmallerLeaves, rendererClasses [&NiagaraMeshRendererProperties]}
 ]
 userVariables [
  User.AxisRotationAngle {type=NiagaraFloat, defaultValue=0, description=""}
  User.BendAngle {type=NiagaraFloat, defaultValue=0, description=""}
  User.FallVector {type=Vector3f, description="", defaultValue {struct=Vector3f, value=<1 0 0>}}
  User.MaxHeight {type=NiagaraFloat, defaultValue=1500, description=""}
  User.MinHeight {type=NiagaraFloat, defaultValue=4000, description=""}
 ]
}
