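# Test material for a destructible physics tree, using broken-trunk base color, mask and normal textures, with the mask's green channel feeding metallic, specular and roughness.
# A BreakTreeDamage value, shaped by an exponent, is compared with vertex color red as a per-vertex threshold.
# Vertices at or below that damage collapse to the object position through World Position Offset, scaled by inverted vertex green.
# Attributes then pass through the shared tree snow and Curie fallback effect functions.
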Material /Game/Environments/Apollo/Foliage/PhysicsTree/Materials/M_PhysicsTreeDestructTest
graph {
 properties {bUseMaterialAttributes=true, bUsedWithNanite=true, pixelDepthOffsetMode=PDOM_Legacy, usedShadingModels=Default Lit}
 nodes [
  n1 = Multiply(
   A=TextureSampleParameter2D() {parameterName=BaseColor, texture=/Game/Environments/Apollo/Foliage/PhysicsTree/Textures/T_PhysicsTree_Trunk_Broken_D}.RGB
   B=VectorParameter() {parameterName=ColorMul_S23, defaultValue=<rgba 1 1 1 1>}.RGB
  )
  n2 = TextureSampleParameter2D() {
   parameterName=Masks
   samplerType=SAMPLERTYPE_Masks
   texture=/Game/Environments/Apollo/Foliage/PhysicsTree/Textures/T_PhysicsTree_Trunk_Broken_S
  }
  n3 = TextureSampleParameter2D() {
   parameterName=Normal
   samplerType=SAMPLERTYPE_Normal
   texture=/Game/Environments/Apollo/Foliage/PhysicsTree/Textures/T_PhysicsTree_Trunk_Broken_N
  }
  n4 = Subtract(A=WorldPosition().XYZ, B=ObjectPositionWS())
  n5 = ScalarParameter() {defaultValue=-1}
  n6 = Power(Base=ScalarParameter() {parameterName=BreakTreeDamage, sliderMax=1}, Exp=ScalarParameter() {defaultValue=1, parameterName=DamageExponent})
  n7 = VertexColor()
  n8 = Constant() {r=1}
  n9 = Constant()
  n10 = Multiply(
   A=LinearInterpolate(B=Multiply(A=@n4, B=@n5), Alpha=If(A=@n6, B=@n7.R, "A > B"=@n8, "A == B"=@n8, "A < B"=@n9)) {constB=500}
   B=OneMinus(None=VertexColor().G)
  )
  MaterialFunctionCall!(
   In=MaterialFunctionCall!(
    In=MakeMaterialAttributes!(
     BaseColor=@n1
     Metallic=@n2.G
     Specular=@n2.G
     Roughness=@n2.G
     Anisotropy
     EmissiveColor
     Opacity
     OpacityMask
     Normal=@n3.RGB
     Tangent
     WorldPositionOffset=@n10
     SubsurfaceColor
     ClearCoat
     ClearCoatRoughness
     AmbientOcclusion
     Refraction
     CustomizedUVs_0
     CustomizedUVs_1
     CustomizedUVs_2
     CustomizedUVs_3
     CustomizedUVs_4
     CustomizedUVs_5
     CustomizedUVs_6
     CustomizedUVs_7
     PixelDepthOffset
     ShadingModel
     Displacement
    )
    FlipTwoSidedSign
    IsTangentSpaceInput=StaticBool() {value=true}
    HasBackFlipFaceNormals
   ) {materialFunction=/Game/Packages/Fortnite_UniqueMaterials_1/MaterialFunctions/Trees/MF_Tree_Snow}.Result
   UseSphericalMappingOverride
  ) {materialFunction=/Game/Packages/Fortress_SharedMaterials/Base_Material_Functions/MF_CurieFallbackEffect}
 ]
}
