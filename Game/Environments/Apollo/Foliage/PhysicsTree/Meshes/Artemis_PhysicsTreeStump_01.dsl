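; Fortnite physics tree stump Blueprint that reports a quest stat event when a player destroys it.
; On server death it calls the parent handler, then casts the owner of a valid damage causer to a player pawn.
; A helper resolves that pawn's controller and quest manager, validates both, and sends a "destroyed tree stump" custom stat event with a count of one.
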(event EventOnDeathServer
  (params (Damage :type Float%20%28single-precision%29) (DamageTags :type Gameplay%20Tag%20Container%20Structure%20%28by%20ref%29) (Momentum :type Vector :default "0, 0, 0") (HitInfo :type Hit%20Result%20Structure%20%28by%20ref%29) (InstigatedBy :type Controller%20Object%20Reference) (DamageCauser :type Actor%20Object%20Reference) (EffectContext :type Gameplay%20Effect%20Context%20Handle%20Structure))
  (|CallParentFunction Damage DamageTags Momentum HitInfo InstigatedBy DamageCauser EffectContext)
  (if (Utilities|IsValid DamageCauser :!class /Script/Engine.KismetSystemLibrary)
    (bind _asfort_player_pawn (Utilities|Casting|CastToFortPlayerPawn (Actor|GetOwner DamageCauser)))
    (CallFunction|SendPlayerKilledStumpStatEvent _asfort_player_pawn)))

(fn SendPlayerKilledStumpStatEvent
  (params (Player :type Fort%20Player%20Pawn%20Object%20Reference))
  (bind _returnvalue (Fortnite|GetFortPlayerControllerfromActor Player))
  (bind _returnvalue_2 (Quests|GetQuestManager _returnvalue "Invalid"))
  (Utilities|IsValid _returnvalue
    (:Is%20Valid
      (Utilities|IsValid _returnvalue_2
        (:Is%20Valid
          (Quest|SendComplexCustomStatEvent _returnvalue_2 0 (LocalVariables|GetStatEventDestroyedTreeTrump) "(GameplayTags=)" 1)
          (return false))
        (:Is%20Not%20Valid
          (return false))))
    (:Is%20Not%20Valid
      (return false))))

(fn ConstructionScript)
